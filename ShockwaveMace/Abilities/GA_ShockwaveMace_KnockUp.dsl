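; Shockwave Mace knock-up ability, extending a knockback parent to damage or launch what the mace hits.
; Activation branches on line of sight, forced-damage tags, physics-prop, team and movement-base checks.
; Launches grant fall-damage immunity and can capsule-trace along the velocity to apply environmental damage to destructibles, skipping traps and spawners.
; Physics props get impulses, auto-aimed by arc prediction at the hostile pawn best aligned with the instigator's facing.
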(event Ability|EventActivateAbility
  (bind _haslos (CallFunction|CheckLOS))
  (if _haslos
    ;[1 1 1] Deal damage unless it is a physics prop we can send flying
    (if (not (CallFunction|ForceDealDamage))
      (if (CallFunction|PhysicsPropCheck)
        _
        (elif (CallFunction|FriendlyCheck)
          _
          (elif (CallFunction|IsBaseOfInstigator)
            (|CallParentFunction)
            (else
              (CallFunction|DealDamage (Ability|GetAvatarActorfromActorInfo))))))
      (else
        _))
    (else
      (bind _returnvalue (FortPhysicsSimulationLibrary|IsPhysicsObjectEmbeddedinLandscape (Ability|GetAvatarActorfromActorInfo))) ; Don't launch if embedded in landscape
      (if _returnvalue
        (CallFunction|DealDamage (Ability|GetAvatarActorfromActorInfo))
        (Utilities|FlowControl|Delay 0.1)
        (Ability|EndAbility)
        (else
          (Ability|EndAbility))))))

(event Ability|EventFailedToActivatePassiveAbility
  (params (ActorInfo :type Gameplay%20Ability%20Actor%20Info%20Structure))
  (|CallParentFunction ActorInfo))

(event EventDoLaunch
  (params (Vel :type Vector :default "0, 0, 0"))
  (|CallParentFunction Vel)
  (GameplayTags|HasAnyMatchingGameplayTags(Message) (Variables|KnockbackParent|GetTarget) (Variables|Default|GetIgnoreFallDamageImmunityTags))
  (Ability|ApplyGameplayEffectToOwner (Variables|Default|GetGE_FallDamageImmunity)) ; Grant fall damage immunity 'til land
  (if (ScalableFloat|GetValueasBool (Variables|Default|GetShouldDestroyEnvironment))
    (bind _outhits (Collision|MultiCapsuleTraceForObjects (+ (Transformation|GetActorLocation (Variables|KnockbackParent|GetTarget)) "0,0,80.000000") (+ (+ (Transformation|GetActorLocation (Variables|KnockbackParent|GetTarget)) "0,0,80.000000") Vel) 8e+01 1.4e+02 (Variables|Default|GetDestroyObject%20Types) false (Utilities|Array|MakeArray (Variables|KnockbackParent|GetTarget)))) ; Trace for destructible building pieces
    (for _array_element _outhits
      (if (Variables|/Script/Engine.Actor|GetbCanBeDamaged (.HitActor (Collision|BreakHitResult _array_element)))
        (bind _bdidactorhavetags (Class|AthenaQuestFunctionLibrary|AthenaCheckActorHasTags (.HitActor (Collision|BreakHitResult _array_element)) "(GameplayTags=((TagName=\"Deimos.Spawner\")))"))
        (if (not _bdidactorhavetags)
          (Utilities|Casting|CastToBuildingTrap (.HitActor (Collision|BreakHitResult _array_element))
            (:CastFailed
              (GameplayEffects|ApplyGameplayEffectToTarget (Ability|GetAbilitySystemComponent (Variables|KnockbackParent|GetTarget)) (Variables|Default|GetGE_EnvironmentalDamage) (Ability|GetAbilitySystemComponent (.HitActor (Collision|BreakHitResult _array_element))) 0.0 (CallFunction|MakeDamageContext)) ; Apply a large amount of environmental damage
            )))))))

(event EventLaunchBuldingActors
  (params (Actor :type Actor%20Object%20Reference))
  (if (and (CallFunction|PhysicsPropCheck) (ScalableFloat|GetValueasBool (Variables|Default|GetImpulsePhysicsObjects)))
    (FortPhysicsSimulationLibrary|AddPhysicsImpulse (Variables|KnockbackParent|GetTarget) (CallFunction|CalculatePhysicsPropLaunch) true "None" true false (Default|GetAbilityInstigator) "(MaxSimSizeAllowed=Large,bRejectEmbeddedInLandscape=False)")
    (else
      _))
  (Ability|EndAbility))

(fn CalculatePhysicsPropLaunch
  (bind _returnvalue (Default|GetAbilityInstigator))
  (bind _returnvalue_2 (Transformation|GetActorLocation _returnvalue))
  (bind _returnvalue_3 (Transformation|GetActorForwardVector _returnvalue))
  (bind _returnvalue_4 (Math|Vector|Normalize _returnvalue_3))
  (bind _returnvalue_5 (Ability|GetAvatarActorfromActorInfo))
  (bind _maxdistance (LocalVariables|GetMaxDistance))
  (if (CallFunction|CanAutoTargetLaunch (Variables|KnockbackParent|GetTarget))
    (bind _outfortpawns (Fortnite|GetFortPawnsInRange (Transformation|GetActorLocation (Variables|KnockbackParent|GetTarget)) (LocalVariables|GetMaxDistance)))
    (for _array_element _outfortpawns
      (if (Utilities|Enum|Equal(Enum) (Team|GetTeamAffiliation _array_element (Default|GetAbilityInstigator)) (LocalVariables|GetHostileAffiliation))
        (bind _asfort_pawn (Utilities|Casting|CastToFortPawn _array_element))
        (Utilities|ContainsHidingActor? _asfort_pawn
          (:false
            (bind (_outlaunchvelocity _returnvalue_6) (Game|SuggestProjectileVelocityCustomArc (Transformation|GetActorLocation (Variables|KnockbackParent|GetTarget)) (+ (FortPawn|GetMovementPredictionPosition _asfort_pawn (+ (/ (Class|AbilityFunctions|GetDistanceBetweenActors _asfort_pawn _returnvalue_5) _maxdistance) 0.1)) "0.000000,0.000000,50.000000") 0.0 0.95))
            (if _returnvalue_6
              (if (and (> (Math|Vector|DotProduct (Math|Vector|Normalize (- (+ (FortPawn|GetMovementPredictionPosition _asfort_pawn (+ (/ (Class|AbilityFunctions|GetDistanceBetweenActors _asfort_pawn _returnvalue_5) _maxdistance) 0.1)) "0.000000,0.000000,50.000000") _returnvalue_2)) _returnvalue_4) (LocalVariables|GetMinDot)) (> (Math|Vector|DotProduct (Math|Vector|Normalize (- (+ (FortPawn|GetMovementPredictionPosition _asfort_pawn (+ (/ (Class|AbilityFunctions|GetDistanceBetweenActors _asfort_pawn _returnvalue_5) _maxdistance) 0.1)) "0.000000,0.000000,50.000000") _returnvalue_2)) _returnvalue_4) (LocalVariables|GetTargetDot)))
                (LocalVariables|SetTargetVelocity _outlaunchvelocity)
                (LocalVariables|SetTargetDot (Math|Vector|DotProduct (Math|Vector|Normalize (- (+ (FortPawn|GetMovementPredictionPosition _asfort_pawn (+ (/ (Class|AbilityFunctions|GetDistanceBetweenActors _asfort_pawn _returnvalue_5) _maxdistance) 0.1)) "0.000000,0.000000,50.000000") _returnvalue_2)) _returnvalue_4))))))))
    (else
      _))
  (return (select (== (LocalVariables|GetTargetDot) 0) (Variables|KnockbackParent|GetFinalLaunchVel) (LocalVariables|GetTargetVelocity))))

(fn CanAutoTargetLaunch
  (params (Target :type Actor%20Object%20Reference))
  (bind _asgameplay_tag_asset_interface (Utilities|Casting|CastToGameplayTagAssetInterface Target)
    (:then
      (return (not (GameplayTags|HasMatchingGameplayTag _asgameplay_tag_asset_interface "(TagName=\"Trait.Immune.ShockwaveHammer.Launch.AutoTargetEnemy\")"))))
    (:CastFailed
      (return true))))

(fn CheckLOS
  (bind _returnvalue (GameplayTags|HasMatchingGameplayTag(Message) (Ability|GetAvatarActorfromActorInfo) "(TagName=\"Building.Type.All\")"))
  (if _returnvalue
    (return true)
    (else
      (bind _returnvalue_2 (Fortnite|CheckLineOfSightToActorWithChannel (Transformation|GetActorLocation (Default|GetAbilityInstigator)) (Ability|GetAvatarActorfromActorInfo) "ECC_GameTraceChannel7" (Default|GetAbilityInstigator)))
      (return _returnvalue_2))))

(fn DealDamage
  (params (Actor :type Actor%20Object%20Reference))
  (GameplayEffects|ApplyGameplayEffectToTarget (Ability|GetFortAbilitySystemComponent (Default|GetAbilityInstigator)) (Variables|Default|GetGE_Damage) (Ability|GetAbilitySystemComponent Actor) 1.0 (CallFunction|MakeDamageContext)))

(fn ForceDealDamage
  (bind _returnvalue (GameplayTags|HasAnyMatchingGameplayTags(Message) (Ability|GetAvatarActorfromActorInfo) (LocalVariables|GetDamageActorsWithTags)))
  (return _returnvalue))

(fn FriendlyCheck
  (bind _returnvalue (Ability|GetAvatarActorfromActorInfo))
  (bind _returnvalue_2 (GameplayTags|HasAnyMatchingGameplayTags(Message) _returnvalue (GameplayTags|MakeLiteralGameplayTagContainer "(GameplayTags=((TagName=\"Building.Type.All\"),(TagName=\"Building.Type.Gameplay\")))")))
  (if _returnvalue_2
    (return false)
    (else
      (return (Utilities|Enum|Equal(Enum) (Team|GetTeamAffiliation _returnvalue (Default|GetAbilityInstigator)) "Friendly")))))

(fn HandleLaunch
  (params (Velocity :type Vector))
  (bind _outfortplayerpawn (CallFunction|GetAvatarasValidFortPlayerPawn)
    (:ValidFortPlayerPawn
      _)
    (:AvatarCastFailed
      _))
  (return (Math|Vector|SelectVector (Variables|GetVelocity) (CallFunction|CalculateVelocityMaintained _outfortplayerpawn (Utilities|Operators|Add (* (Math|Vector|Normalize (Pawn|Components|CharacterMovement|GetCurrentAcceleration (Variables|/Script/Engine.Character|GetCharacterMovement _outfortplayerpawn))) (ScalableFloat|GetValueatLevel (Variables|KnockbackParent|GetRow_XYKnockMag))) (Math|Vector|MakeVector 0.0 0.0 (.z (Variables|GetVelocity))))) (== _outfortplayerpawn (Utilities|Struct|BreakGameplayEventData (Variables|Ability|GetCurrentEventData)))) (== _outfortplayerpawn (Utilities|Struct|BreakGameplayEventData (Variables|Ability|GetCurrentEventData))) _outfortplayerpawn))

(fn HasPhysicsComponent
  (params (Actor :type Actor%20Object%20Reference))
  (Utilities|IsValid (Actor|GetComponentbyClass Actor "/Script/FortniteGame.FortPhysicsObjectComponent")
    (:Is%20Valid
      (return true))
    (:Is%20Not%20Valid
      (return false))))

(fn IsBaseOfInstigator
  (bind _asfort_player_pawn (Utilities|Casting|CastToFortPlayerPawn (Default|GetAbilityInstigator))
    (:then
      (return (== (Pawn|GetMovementBaseActor _asfort_player_pawn) (Ability|GetAvatarActorfromActorInfo))))
    (:CastFailed
      (return false))))

(fn MakeDamageContext
  (bind _returnvalue (Default|GetAbilityInstigator))
  (bind _asfort_pawn (Utilities|Casting|CastToFortPawn _returnvalue)
    (:then
      (return (Class|AthenaFunctionLibrary|MakeSimpleFortEffectContext _returnvalue (select (Utilities|IsValid (Weapon|GetCurrentWeapon _asfort_pawn) :!class /Script/Engine.KismetSystemLibrary) (Weapon|GetCurrentWeapon _asfort_pawn) _returnvalue))))
    (:CastFailed
      ;[0.7812 0.2326 0 0.5] If the instigator isn't a FortPawn
      (return (Class|AthenaFunctionLibrary|MakeSimpleFortEffectContext (Default|GetAbilityInstigator) (Default|GetAbilityInstigator))))))

(fn PhysicsPropCheck
  (bind _returnvalue (Ability|GetAvatarActorfromActorInfo))
  (bind _returnvalue_2 (Ability|GetAvatarActorfromActorInfo))
  ;[1 1 1] If it's a device, ensure it wants to be moved
  (bind _returnvalue_3 (GameplayTags|HasMatchingGameplayTag(Message) _returnvalue_2 "(TagName=\"Asset.Creative.Device\")"))
  (if _returnvalue_3
    (bind _returnvalue_4 (GameplayTags|HasMatchingGameplayTag(Message) _returnvalue_2 "(TagName=\"Building.Type.Physics\")"))
    (if _returnvalue_4
      (bind _returnvalue_5 (GameplayTags|HasMatchingGameplayTag(Message) _returnvalue "(TagName=\"Building.Type.IgnorePhysics\")"))
      (if (not _returnvalue_5)
        (bind _hasphysicscomponent (CallFunction|HasPhysicsComponent _returnvalue))
        (bind _returnvalue_6 (FortPhysicsSimulationLibrary|IsPhysicsObjectEmbeddedinLandscape _returnvalue)) ; Don't launch if embedded in landscape
        (return (and (not _returnvalue_6) _hasphysicscomponent)))
      (else
        _))
    (return false)
    (else
      _)))

(fn SendQuestEventToInstigator
  (bind _asfort_player_pawn (Utilities|Casting|CastToFortPlayerPawn (Variables|KnockbackParent|GetTarget)))
  (if (Utilities|Enum|Equal(Enum) (Team|GetTeamAffiliation _asfort_player_pawn (Default|GetAbilityInstigator)) "Hostile")
    (Quest|SendComplexCustomStatEvent (Quests|GetQuestManager (Fortnite|GetFortPlayerControllerfromActor (Default|GetAbilityInstigator)) "Athena") 0 "(GameplayTags=)" (Variables|Default|GetLaunchOpponentQuestTag) 1)))
