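; Shockwave Mace self knock-up ability that launches the player, built on a knockback parent.
; It overrides DoLaunch to steer horizontal velocity by movement input, keeping vertical speed, before calling the parent.
; It then sends a bounce event for quests and grants fall-damage immunity unless ignore tags match.
; A capsule trace along the launch vector applies a build-destroying GameplayEffect to damageable actors, excluding traps and Deimos spawners.
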(event EventDoLaunch
  (params (Vel :type Vector :default "0, 0, 0"))
  (bind _returnvalue (CallFunction|HandleSelfLaunchInput Vel))
  (|CallParentFunction _returnvalue)
  ;[1 1 1] Quest for bouncing
  (Ability|SendGameplayEventtoActor (Ability|GetAvatarActorfromActorInfo) "(TagName=\"Event.ShockwaveMace.Trigger.Bounce.Launch\")")
  (bind _returnvalue_2 (GameplayTags|HasAnyMatchingGameplayTags(Message) (Variables|KnockbackParent|GetTarget) (Variables|Default|GetIgnoreFallDamageImmunityTags)))
  (if (not _returnvalue_2)
    (Ability|ApplyGameplayEffectToOwner (Variables|Default|GetGE_FallDamageImmunity)) ; Grant fall damage immunity 'til land
  )
  (bind _outhits (Collision|MultiCapsuleTraceForObjects (+ (Transformation|GetActorLocation (Variables|KnockbackParent|GetTarget)) "0,0,80.000000") (+ (+ (Transformation|GetActorLocation (Variables|KnockbackParent|GetTarget)) "0,0,80.000000") Vel) 8e+01 1.4e+02 (Variables|Default|GetDestroyObject%20Types) false (Utilities|Array|MakeArray (Variables|KnockbackParent|GetTarget)))) ; Trace for destructible building pieces
  (for _array_element _outhits
    (if (Variables|/Script/Engine.Actor|GetbCanBeDamaged (.HitActor (Collision|BreakHitResult _array_element)))
      (bind _bdidactorhavetags (Class|AthenaQuestFunctionLibrary|AthenaCheckActorHasTags (.HitActor (Collision|BreakHitResult _array_element)) "(GameplayTags=((TagName=\"Deimos.Spawner\")))"))
      (if (not _bdidactorhavetags)
        (Utilities|Casting|CastToBuildingTrap (.HitActor (Collision|BreakHitResult _array_element))
          (:CastFailed
            (GameplayEffects|ApplyGameplayEffectToTarget :self (Ability|GetAbilitySystemComponent (Variables|KnockbackParent|GetTarget)) :GameplayEffectClass "/ShockwaveMace/GameplayEffects/GE_ShockwaveMace_DestroyBuilds.GE_ShockwaveMace_DestroyBuilds_C" :Target (Ability|GetAbilitySystemComponent (.HitActor (Collision|BreakHitResult _array_element))) :Level 0.0) ; Apply environmental damage
          ))))))

(fn HandleSelfLaunchInput
  (params (Velocity :type Vector))
  (bind _outfortplayerpawn (CallFunction|GetAvatarasValidFortPlayerPawn)
    (:ValidFortPlayerPawn
      _)
    (:AvatarCastFailed
      _))
  (return (CallFunction|CalculateVelocityMaintained _outfortplayerpawn (Utilities|Operators|Add (* (Math|Vector|Normalize (Pawn|Components|CharacterMovement|GetCurrentAcceleration (Variables|/Script/Engine.Character|GetCharacterMovement _outfortplayerpawn))) (ScalableFloat|GetValueatLevel (Variables|KnockbackParent|GetRow_XYKnockMag))) (Math|Vector|MakeVector 0.0 0.0 (.z (Variables|GetVelocity)))))))
